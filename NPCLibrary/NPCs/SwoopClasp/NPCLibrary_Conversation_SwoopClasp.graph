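# Conversation database for the SwoopClasp NPC (General Onyx Winter), defining their greeting and main conversation hub.
# The entry point leads to a speech node that swaps its default line for contextual ones on a first-ever meeting, a matching conversation count, or a specific player character cosmetic.
# The speech node then branches through link nodes to NPC library quests, basic services, a Season 37 story quest and SwoopClasp's own dialogue entry.
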ConversationDatabase /NPCLibrary/NPCs/SwoopClasp/NPCLibrary_Conversation_SwoopClasp
entry_points [{tag=Conversation.Entry.SwoopClasp, destinations [#1]}]
nodes [
 ConversationEntryPointNode_7 {
  id=#1
  class=ConversationEntryPointNode
  outputs [#2]
  properties {entryTag {tagName=Conversation.Entry.SwoopClasp}}
 }
 Speech_C_4 {
  id=#2
  class=Speech_C
  outputs [#3, #4, #5, #6]
  properties {
   nodeName=""
   generalConfig {
    defaultMessage=This Island is ours.
    contextualMessages [
     {
      message="General Onyx Winter, at your service. As long as you don't make me mad."
      contextRequirements [{requirement {class=FortControllerRequirement_IsFirstEverConversationWithNPC}}]
      priority {curve {rowName=Tandem.Speech.Priority.FirstEverInteraction}}
     }
     {
      message=This Island is ours.
      contextRequirements [{requirement {class=FortControllerRequirement_CheckNPCMatchConversationCount, properties {nPCUniqueIdTag {tagName=AISpawnerData.Type.Tandem.SwoopClasp}}}}]
      priority {curve {curveTable=null, rowName=None}}
     }
     {
      message=Your field reports arrived covered in Bug detritus again. Heh.
      contextRequirements [{requirement {class=FortControllerRequirement_HasCID, properties {allowedCIds [/BRCosmetics/Athena/Items/Cosmetics/Characters/Character_TalonPane_Flail]}}}]
      priority {value=2, curve {curveTable=null, rowName=None}}
     }
    ]
   }
  }
 }
 ConversationLinkNode_3 {
  id=#6
  class=ConversationLinkNode
  properties {remoteEntryTag {tagName=Conversation.Entry.SwoopClasp.Dialogue}}
 }
 ConversationLinkNode_2 {
  id=#5
  class=ConversationLinkNode
  properties {remoteEntryTag {tagName=Conversation.Entry.Season37.Story.P01.Q02}}
 }
 ConversationLinkNode_0 {id=#4, class=ConversationLinkNode, properties {remoteEntryTag {tagName=Conversation.Entry.BasicServices}}}
 ConversationLinkNode_1 {id=#3, class=ConversationLinkNode, properties {remoteEntryTag {tagName=Conversation.Entry.NPCLibrary.Quests}}}
]
