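# Conversation database for the SwoopClasp NPC's dialogue, opened through its Conversation.Entry.SwoopClasp.Dialogue entry tag.
# An opening greeting under an "Ask about..." choice branches to a Combat topic, a Hope and Jones topic, and a Back option.
# Each topic is a linear chain of speech nodes covering the character's combat style or O.X.R.'s support against the Dark Presence.
# Both chains converge on a shared return-to-start node labelled BACK.
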ConversationDatabase /NPCLibrary/NPCs/SwoopClasp/NPCLibrary_Conversation_SwoopClasp_Dialogue
entry_points [{tag=Conversation.Entry.SwoopClasp.Dialogue, destinations [#1]}]
nodes [
 ConversationEntryPointNode_7 {
  id=#1
  class=ConversationEntryPointNode
  outputs [#2]
  properties {entryTag {tagName=Conversation.Entry.SwoopClasp.Dialogue}}
 }
 Speech_C_0 {
  id=#2
  class=Speech_C
  outputs [#3, #4, #5]
  properties {
   nodeName=""
   softTaskIcon=&T_UI_NPC_question
   generalConfig {defaultMessage=Make it quick. I've got a war to fight.}
  }
  sub_nodes [ConversationChoiceNode_10 {class=ConversationChoiceNode, properties {defaultChoiceDisplayText=Ask about...}}]
 }
 Speech_C_2 {
  id=#5
  class=Speech_C
  outputs [#6]
  properties {
   nodeName=""
   softTaskIcon=&T_UI_NPC_question
   generalConfig {defaultMessage="Now that the Swarm's gone, I've had time to look over Hope's findings."}
  }
  sub_nodes [ConversationChoiceNode_14 {class=ConversationChoiceNode, properties {defaultChoiceDisplayText=Hope and Jones}}]
 }
 Speech_C_3 {
  id=#6
  class=Speech_C
  outputs [#7]
  properties {nodeName="", generalConfig {defaultMessage="Until this \"Dark Presence\" is defeated, the Swarm's return remains a possibility."}}
 }
 Speech_C_1 {
  id=#7
  class=Speech_C
  outputs [#8]
  properties {nodeName="", generalConfig {defaultMessage="I'm assigning O.X.R. assets to their command and helping how I can, too."}}
 }
 Speech_C_11 {
  id=#8
  class=Speech_C
  outputs [#9]
  properties {nodeName="", generalConfig {defaultMessage=O.X.R.'s greatest success would be its own abolition. I'm hoping I'll get to see it someday.}}
 }
 ReturnToStart_C_23 {
  id=#9
  class=ReturnToStart_C
  properties {softTaskIcon=&T_UI_NPC_Back}
  sub_nodes [ConversationChoiceNode_12 {class=ConversationChoiceNode, properties {defaultChoiceDisplayText=BACK}}]
 }
 Back_C_1 {
  id=#4
  class=Back_C
  sub_nodes [ConversationChoiceNode_13 {class=ConversationChoiceNode, properties {defaultChoiceDisplayText=BACK}}]
 }
 Speech_C_7 {
  id=#3
  class=Speech_C
  outputs [#10]
  properties {
   nodeName=""
   softTaskIcon=&T_UI_NPC_question
   generalConfig {defaultMessage="Ah, I miss the field. I'm a great tactician, sure. Some say the best. But my combat style..."}
  }
  sub_nodes [ConversationChoiceNode_11 {class=ConversationChoiceNode, properties {defaultChoiceDisplayText=Combat}}]
 }
 Speech_C_15 {
  id=#10
  class=Speech_C
  outputs [#11]
  properties {nodeName="", generalConfig {defaultMessage="...Well, it's been described as \"animalistic\" and \"brutal.\" "}}
 }
 Speech_C_14 {
  id=#11
  class=Speech_C
  outputs [#12]
  properties {nodeName="", generalConfig {defaultMessage="What can I say? I enjoy my work. "}}
 }
 Speech_C_5 {
  id=#12
  class=Speech_C
  outputs [#9]
  properties {
   nodeName=""
   generalConfig {defaultMessage="But my gift for combat is second to my tactical prowess. So I serve behind a desk, and live vicariously through my people."}
  }
 }
]
